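# Animation blueprint for the SilverBellWind female head that shifts the under-chin facial bones as the jaw opens.
# Each update, the event graph reads the jaw-open expression curve from the part skeletal mesh's anim instance into JawOpen.
# The AnimGraph chains additive bone-space translations on the center, right and left UnderChin bones, using JawOpen as alpha to scale the offset.
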AnimBlueprint /BRCosmetics/Characters/Player/Female/Medium/Heads/F_MED_SilverBellWind_Head/Meshes/F_MED_SilverBellWind_UnderChin_Transform_AnimBP1
target_skeleton=/Game/Characters/Player/Common/Fortnite_Base_Head/Fortnite_Base_3L_Head_Skeleton
graphs [
 AnimGraph {nodes [
   n1 = Get JawOpen() -> (JawOpen:double)
   n2 = "Transform (Modify) Bone - Bone: FACIAL_C_UnderChin"(ComponentPose=Local To Component(LocalPose=Input Pose()), Alpha=@n1, Translation="0,-3,-5", Rotation="0,0,0", Scale="1,1,1") {
    translationMode=BMM_Additive
    translationSpace=BCS_BoneSpace
    alphaBoolBlend {blendOption=Linear}
    boneToModify {boneName=FACIAL_C_UnderChin}
    translation {y=-3, z=-5}
   }
   n3 = "Transform (Modify) Bone - Bone: FACIAL_R_UnderChin"(ComponentPose=@n2, Alpha=@n1, Translation="0,-3,-5", Rotation="0,0,0", Scale="1,1,1") {
    translationMode=BMM_Additive
    translationSpace=BCS_BoneSpace
    alphaBoolBlend {blendOption=Linear}
    boneToModify {boneName=FACIAL_R_UnderChin}
    translation {y=-3, z=-5}
   }
   n4 = "Transform (Modify) Bone - Bone: FACIAL_L_UnderChin"(ComponentPose=@n3, Alpha=@n1, Translation="0,-3,-5", Rotation="0,0,0", Scale="1,1,1") {
    translationMode=BMM_Additive
    translationSpace=BCS_BoneSpace
    alphaBoolBlend {blendOption=Linear}
    boneToModify {boneName=FACIAL_L_UnderChin}
    translation {y=-3, z=-5}
   }
   Output Pose(Result=Component To Local(ComponentPose=@n4)) {layerGroup=None}
  ]}
 EventGraph {nodes [
   Event BlueprintUpdateAnimation()
   n1 = GetPartSkeletalMeshComponent()
   > Set JawOpen(JawOpen:double=GetCurveValue(self=GetAnimInstance(self=@n1), CurveName=CTRL_expressions_jawOpen)) -> (Output_Get:double)
  ]}
]
variables [
 JawOpen {type=number}
 Out Names {type="array<string>", default []}
]
